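# Two-sided lit material for Niagara mesh particles that renders distant seagulls with procedurally animated wings in the Flood environment effects.
# A sine of the particle's FlapTime, masked by vertex color and UV1, rotates vertices about a local axis, with delayed tip bend and an added VerticalWave offset.
# The normal map gets a matching rotation, while particle color and a two-color gradient tint the texture.
# A debug parameter substitutes engine time and actor position.
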Material /Game/Effects/Fort_Effects/Materials/Environment/Flood/M_Distant_Bird_Particle
graph {
 properties {bUsedWithNiagaraMeshParticles=true, pixelDepthOffsetMode=PDOM_Legacy, twoSided=true, usedShadingModels=Default Lit}
 nodes [
  n1 = Constant3Vector() {constant=<rgba 1 0 -0.25 0>}
  n2 = Constant3Vector() {constant=<rgba 1 0 0.12 0>}
  n3 = NamedRerouteUsage()
  n4 = Normalize(VectorInput=Transform(None=LinearInterpolate(A=@n1, B=@n2, Alpha=@n3)) {transformSourceType=TRANSFORMSOURCE_Local})
  n5 = VertexColor()
  n6 = TextureCoordinate() {coordinateIndex=1}
  n7 = LinearInterpolate(Alpha=MaterialFunctionCall(In=ComponentMask(None=@n6) {r=true}, Contrast=ScalarParameter() {defaultValue=5, parameterName=FlapMaskContrast}) {materialFunction=&CheapContrast}.Result) {constA=-1}
  n8 = DynamicParameter() {paramNames [FlapAnimMult,FlapTime,Param3,Param4]}
  n9 = Time()
  n10 = ScalarParameter() {parameterName=debug}
  n11 = ComponentMask(None=TextureCoordinate() {coordinateIndex=1}) {g=true}
  n12 = ScalarParameter() {defaultValue=-0.07, parameterName=Top/Bottom Wing Offset}
  n13 = Add(
   A=NamedRerouteDeclaration(None=LinearInterpolate(A=@n8.FlapTime, B=@n9, Alpha=@n10)) {name=FlapTime, nodeColor=<rgba 1 0.0507 0.1004 1>}
   B=Multiply(A=Multiply(A=@n11, B=@n12), B=NamedRerouteUsage())
  )
  n14 = ConstantBiasScale(None=Sine(None=@n13))
  n15 = LinearInterpolate(A=ScalarParameter() {defaultValue=-0.22, parameterName=FlapMin}, B=ScalarParameter() {defaultValue=0.16, parameterName=FlapMax}, Alpha=@n14)
  n16 = ScalarParameter() {defaultValue=-0.25, parameterName=EdgeFlapTimeOfsset}
  n17 = TextureCoordinate() {coordinateIndex=1}
  n18 = OneMinus(None=SphereMask(A=ComponentMask(None=@n17) {r=true}, B=Constant() {r=0.5}, Radius=Constant() {r=0.5}) {attenuationRadius=0.4, hardnessPercent=0})
  n19 = ScalarParameter() {defaultValue=0.07, parameterName=WingBendDown}
  n20 = ScalarParameter() {parameterName=WingBendUp}
  n21 = Power(Base=NamedRerouteUsage()) {constExponent=5}
  n22 = Multiply(
   A=Sine(None=Add(A=@n13, B=@n16) {constB=0.25})
   B=Multiply(A=Power(Base=@n18) {constExponent=3}, B=LinearInterpolate(A=@n19, B=@n20, Alpha=@n21) {constB=0.05}) {constB=0.1}
  )
  n23 = Multiply(A=NamedRerouteUsage(), B=Multiply(A=Multiply(A=@n5.R, B=@n7), B=Add(A=@n15, B=@n22)))
  n24 = ParticlePositionWS()
  n25 = ActorPositionWS()
  n26 = ScalarParameter() {parameterName=debug}
  n27 = TextureCoordinate() {coordinateIndex=1}
  n28 = Add(A=NamedRerouteUsage()) {constB=0.18}
  n29 = Sine(None=Add(A=ComponentMask(None=@n27) {g=true}, B=Multiply(A=@n28) {constB=10})) {period=10}
  n30 = LinearInterpolate(Alpha=Power(Base=ConstantBiasScale(None=@n29)) {constExponent=0.7}) {constA=-1}
  n31 = Transform(None=VectorParameter() {parameterName=VerticalWave, defaultValue=<rgba 0 0 6 1>}.RGB) {transformSourceType=TRANSFORMSOURCE_Local}
  Add(
   A=RotateAboutAxis(
    NormalizedRotationAxis=NamedRerouteDeclaration(None=@n4) {name=RotationAxis, nodeColor=<rgba 0 0.5036 1 1>}
    RotationAngle=NamedRerouteDeclaration(None=@n23) {name=RotationAngle, nodeColor=<rgba 0 0.5036 1 1>}
    PivotPoint=LinearInterpolate(A=@n24, B=@n25, Alpha=@n26)
    Position=WorldPosition().XYZ
   )
   B=Multiply(A=Multiply(A=@n30, B=@n31), B=NamedRerouteUsage())
  )
  n32 = TextureSampleParameter2D() {parameterName=ColorTexture, texture=/Game/Effects/Fort_Effects/Textures/Environments/Flood/T_Seagull_White_D}
  LinearInterpolate(
   A=ScalarParameter() {defaultValue=1, parameterName=RoughnessHigh}
   B=ScalarParameter() {defaultValue=0.5, parameterName=RoughnessLOW}
   Alpha=@n32.R
  )
  n33 = TextureCoordinate() {coordinateIndex=1}
  Multiply(
   A=LinearInterpolate(
    A=VectorParameter() {parameterName=Color2, defaultValue=<rgba 0.3 0.3 0.3 1>}.RGB
    B=VectorParameter() {parameterName=Color1, defaultValue=<rgba 0.515 0.515 0.515 1>}.RGB
    Alpha=ComponentMask(None=@n33) {r=true}
   )
   B=Multiply(A=@n32.RGB, B=ParticleColor().RGB)
  )
  NamedRerouteDeclaration(None=@n8.FlapAnimMult) {name=FlapAnimMult, nodeColor=<rgba 0 1 0.625 1>}
  NamedRerouteDeclaration(None=@n14) {name=Flap0-1, nodeColor=<rgba 0.1484 0 1 1>}
  n34 = TextureSampleParameter2D() {
   parameterName=NormalTexture
   samplerType=SAMPLERTYPE_Normal
   texture=/Game/Effects/Fort_Effects/Textures/Environments/Flood/T_Seagull_Seagull_N
  }
  n35 = Transform(None=@n34.RGB) {transformType=TRANSFORM_ParticleWorld}
  n36 = RotateAboutAxis(NormalizedRotationAxis=NamedRerouteUsage(), RotationAngle=NamedRerouteUsage(), PivotPoint=Constant(), Position=@n35) {desc=wo}
  Transform(None=Normalize(VectorInput=Add(A=@n36, B=@n35))) {transformSourceType=TRANSFORMSOURCE_ParticleWorld, transformType=TRANSFORM_Tangent}
 ]
}
