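# Lit material for a seagull mesh used as Niagara mesh particles in flood environment effects, animating wing flapping procedurally in the vertex math.
# Time-driven sines combine a local-Z wave with rotations about local X and Y, masked by vertex color channels.
# Dynamic parameters supply per-particle timing bias, speed and an extra vertical bob.
# Color multiplies the bird texture by particle color and a UV1-blended two-color tint, also feeding an emissive adjustment function.
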Material /Game/Effects/Fort_Effects/Materials/Environment/Flood/M_Seagull_01
graph {
 properties {
  bFullyRough=true
  bTangentSpaceNormal=false
  bUsedWithNiagaraMeshParticles=true
  bUsedWithStaticLighting=true
  numCustomizedUVs=2
  pixelDepthOffsetMode=PDOM_Legacy
  previewMesh=/Game/Effects/Fort_Effects/Mesh/Birds/S_Seagull_01
  usedShadingModels=Default Lit
 }
 nodes [
  n1 = Transform(None=Constant3Vector() {constant=<rgba 0 0 1 0>}) {transformSourceType=TRANSFORMSOURCE_Local}
  n2 = ComponentMask(None=Divide(A=LocalPosition().XYZ) {constB=84}) {g=true}
  n3 = MaterialFunctionCall(In=Power(Base=Abs(None=@n2))) {materialFunction=&CheapContrast}
  n4 = LinearInterpolate(B=OneMinus(None=Saturate(None=@n3.Result)), Alpha=ScalarParameter() {parameterName=UseFlapOffset})
  n5 = Time()
  n6 = DynamicParameter() {defaultValue=<rgba 0 0 0 1>, paramNames [Bias,Speed,Param3,Param4]}
  n7 = Multiply(A=Add(A=Add(A=@n4, B=@n5), B=@n6.Bias), B=@n6.Speed)
  n8 = ScalarParameter() {parameterName=Macro Time Bias}
  n9 = Transform(None=WorldPosition().XYZ) {transformSourceType=TRANSFORMSOURCE_World, transformType=TRANSFORM_Local}
  n10 = ScalarParameter() {defaultValue=128, parameterName=Macro Wave Frequency}
  n11 = ScalarParameter() {defaultValue=128, parameterName=Macro Wave Frequency Wings}
  n12 = Multiply(A=VertexColor().A, B=ScalarParameter() {parameterName=UseMacroWingFreq?})
  n13 = Add(
   A=Multiply(A=Add(A=@n7, B=@n8), B=ScalarParameter() {defaultValue=1, parameterName=Speed})
   B=Divide(A=ComponentMask(None=@n9) {r=true}, B=LinearInterpolate(A=@n10, B=@n11, Alpha=@n12))
  )
  n14 = Multiply(A=Multiply(A=Sine(None=@n13), B=ScalarParameter() {defaultValue=1, parameterName=Macro Wave scale}), B=VertexColor().B)
  n15 = Constant3Vector() {constant=<rgba 1 0 0 0>}
  n16 = VertexColor()
  n17 = Sine(None=Multiply(A=@n7, B=ScalarParameter() {defaultValue=1, parameterName=Speed}))
  n18 = ScalarParameter() {parameterName=Rotate Amount}
  n19 = Multiply(A=Subtract(A=Multiply(A=@n16.R) {constB=2}), B=Add(A=Multiply(A=@n17, B=@n18), B=ScalarParameter() {parameterName=Rotation Angle Bias}))
  n20 = Constant3Vector() {constant=<rgba 0 1 0 0>}
  n21 = Subtract(A=Multiply(A=VertexColor().G) {constB=2})
  n22 = Add(A=ScalarParameter() {parameterName=Forward Rotate Delay}, B=@n7)
  n23 = ScalarParameter() {defaultValue=1, parameterName=Speed}
  n24 = Multiply(A=Sine(None=Multiply(A=@n22, B=@n23)), B=ScalarParameter() {parameterName=Forward rotate Amount})
  n25 = Add(
   A=RotateAboutAxis(
    NormalizedRotationAxis=Transform(None=@n15) {transformSourceType=TRANSFORMSOURCE_Local}
    RotationAngle=Add(B=@n19)
    PivotPoint=ObjectPositionWS()
    Position=WorldPosition().XYZ
   )
   B=RotateAboutAxis(
    NormalizedRotationAxis=Transform(None=@n20) {transformSourceType=TRANSFORMSOURCE_Local}
    RotationAngle=Multiply(A=@n21, B=@n24)
    PivotPoint=ObjectPositionWS()
    Position=WorldPosition().XYZ
   )
  )
  n26 = VertexColor()
  n27 = DynamicParameter() {defaultValue=<rgba 0 0 0 1>, paramNames [Bias,Speed,Param3,Param4]}
  n28 = Sine(None=Multiply(A=@n27.Param4, B=Time())) {period=3}
  n29 = Constant3Vector() {constant=<rgba 0 0 1 0>}
  Add(A=Add(A=Multiply(A=@n1, B=@n14), B=Multiply(A=@n25, B=@n26.A)), B=Multiply(A=@n27.Param3, B=Multiply(A=@n28, B=@n29)))
  n30 = CollectionParameter() {
   collection=&FortniteMaterialParameters
   parameterName=SunlightDirectionalVector
  }
  Clamp(None=DotProduct(A=ComponentMask(None=@n30) {b=true, g=true, r=true}, B=VertexNormalWS()))
  Constant()
  n31 = Constant() {r=1}
  If(A=@n6.Speed, B=Constant(), "A > B"=@n6.Speed, "A == B"=@n31, "A < B"=@n31)
  n32 = TextureSampleParameter2D() {parameterName=ColorTexture, texture=/Game/Effects/Fort_Effects/Textures/Environments/Birds/T_Zen_Bird_D}
  LinearInterpolate(
   A=ScalarParameter() {defaultValue=1, parameterName=RoughnessHigh}
   B=ScalarParameter() {defaultValue=0.5, parameterName=RoughnessLOW}
   Alpha=@n32.R
  )
  n33 = ParticleColor()
  n34 = LinearInterpolate(
   A=VectorParameter() {parameterName=Color2, defaultValue=<rgba 0.3 0.3 0.3 1>}.RGB
   B=VectorParameter() {parameterName=Color1, defaultValue=<rgba 0.515 0.515 0.515 1>}.RGB
   Alpha=ComponentMask(None=TextureCoordinate() {coordinateIndex=1}) {r=true}
  )
  n35 = Multiply(A=@n32.RGB, B=ParticleColor().RGB)
  n36 = Multiply(A=@n34, B=@n35)
  MaterialFunctionCall(Emissive=Add(A=ScalarParameter() {parameterName=Emissive}, B=Multiply(A=@n33.RGB, B=@n36))) {materialFunction=&MF_FX_EmissiveAdjustments}
  Multiply(A=@n34, B=@n35)
  TextureSampleParameter2D() {parameterName=NormalTexture, samplerType=SAMPLERTYPE_Normal, texture=/Game/Effects/Fort_Effects/Textures/Environments/Birds/T_Zen_Bird_N}
  VectorParameter() {parameterName=Color, defaultValue=<rgba 1 1 1 1>}
 ]
}
